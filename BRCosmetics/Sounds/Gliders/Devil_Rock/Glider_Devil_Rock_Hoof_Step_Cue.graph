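# Sound cue for hoof step sounds on the Devil Rock glider cosmetic.
# It picks one of six equally weighted hoof samples at random, then a modulator slightly varies the volume.
# Playback uses the glider thrust attenuation and in-air glider sound class at reduced volume, with a concurrency override capping simultaneous instances.
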SoundCue /BRCosmetics/Sounds/Gliders/Devil_Rock/Glider_Devil_Rock_Hoof_Step_Cue
root=Modulator_0
nodes [
 {id=Modulator_0, children [Random_1], properties {volumeMax=1, volumeMin=0.85}}
 {
  id=Random_1
  children [WavePlayer_0, WavePlayer_7, WavePlayer_8, WavePlayer_9, WavePlayer_10, WavePlayer_11]
  properties {weights [1,1,1,1,1,1]}
 }
 {id=WavePlayer_0, properties {soundWaveAssetPtr=&/Glider_Devil_Rock_Hoof_06}}
 {id=WavePlayer_10, properties {soundWaveAssetPtr=&/Glider_Devil_Rock_Hoof_02}}
 {id=WavePlayer_11, properties {soundWaveAssetPtr=&/Glider_Devil_Rock_Hoof_01}}
 {id=WavePlayer_7, properties {soundWaveAssetPtr=&/Glider_Devil_Rock_Hoof_05}}
 {id=WavePlayer_8, properties {soundWaveAssetPtr=&/Glider_Devil_Rock_Hoof_04}}
 {id=WavePlayer_9, properties {soundWaveAssetPtr=&/Glider_Devil_Rock_Hoof_03}}
]
properties {
 attenuationSettings=&Fort_Player_Glider_Thrust_Athena_Att
 bOverrideConcurrency=true
 maxDistance=12500
 soundClassObject=&Fort_Gliders_InAir
 volumeMultiplier=0.25
 concurrencyOverrides {maxCount=4}
}
