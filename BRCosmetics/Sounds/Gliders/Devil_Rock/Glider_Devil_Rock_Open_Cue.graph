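# Sound cue for the opening sound of the Devil Rock glider cosmetic.
# A fixed short Delay node feeds a single Wave Player that plays the one open sample.
# Attenuation, audio property sheet, concurrency and sound class come from shared glider open/close assets, with a slight volume boost.
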SoundCue /BRCosmetics/Sounds/Gliders/Devil_Rock/Glider_Devil_Rock_Open_Cue
root=Delay_0
nodes [
 {id=Delay_0, children [WavePlayer_0], properties {delayMax=0.1, delayMin=0.1}}
 {id=WavePlayer_0, properties {soundWaveAssetPtr=&/Glider_Devil_Rock_Open_01}}
]
properties {
 attenuationSettings=/Game/Sounds/Attenuation/Fort_Player_Glider_OpenClose_Athena_Att
 audioPropertiesSheet=/BRCosmetics/Sounds/Settings/PropertySheets/Glider/APS_Cosmetic_Glider_Open
 maxDistance=10000
 soundClassObject=&Fort_Gliders_Land
 volumeMultiplier=1.25
 concurrencySet [&Athena_GliderOpenClose_Conc]
}
