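# Sound cue for the roar of the Devil Rock glider cosmetic.
# A random node picks one of two roar wave variations with equal weight.
# It plays at reduced volume through the shared glider thrust attenuation and the glider land sound class.
# Concurrency is overridden to cap simultaneous instances and block new ones, alongside a glider open/close concurrency set.
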SoundCue /BRCosmetics/Sounds/Gliders/Devil_Rock/Glider_Devil_Rock_Roar_Cue
root=Random_0
nodes [
 {id=Random_0, children [WavePlayer_0,WavePlayer_2], properties {weights [1,1]}}
 {id=WavePlayer_0, properties {soundWaveAssetPtr=&/Glider_Devil_Rock_Roar_02}}
 {id=WavePlayer_2, properties {soundWaveAssetPtr=&/Glider_Devil_Rock_Roar_01}}
]
properties {
 attenuationSettings=&Fort_Player_Glider_Thrust_Athena_Att
 bOverrideConcurrency=true
 maxDistance=12500
 soundClassObject=&Fort_Gliders_Land
 volumeMultiplier=0.25
 concurrencyOverrides {maxCount=2, resolutionRule=PreventNew}
 concurrencySet [&Athena_GliderOpenClose_Conc]
}
